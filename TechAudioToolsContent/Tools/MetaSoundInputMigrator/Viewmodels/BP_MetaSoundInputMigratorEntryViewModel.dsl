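; Viewmodel for an entry in the MetaSound Input Migrator tool, exposing the name of the input it is linked to as display text.
; Its single function follows the linked migrator entry viewmodel to that entry's data and casts it to a MetaSound input editor viewmodel.
; It converts the input name to Text, with no name lookup when the link or entry data is missing.
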(fn GetLinkedInputNameasText
  (bind _linked_migrator_entry_viewmodel (Variables|Default|GetLinked%20Migrator%20Entry%20Viewmodel)
    (:then
      (bind _entrydata (Variables|/Script/TechAudioToolsContent.TechAudioToolsEntryViewModel|GetEntryData _linked_migrator_entry_viewmodel)
        (:then
          (bind _asmetasound_input_editor_viewmodel (Utilities|Casting|CastToMetaSoundInputEditorViewModel _entrydata))
          (return (Utilities|Text|ToText(Name) (Variables|/Script/TechAudioToolsMetaSound.MetaSoundInputViewModel|GetInputName _asmetasound_input_editor_viewmodel))))
        (:else)))
    (:else)))
