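# Editor utility viewmodel for the MetaSound Input Migrator, which copies input defaults from a source MetaSound to a destination.
# It links type-matched input pairs, mirrors links on list entry viewmodels, and broadcasts field changes for UI state.
# Builder APIs apply differing linked defaults to the destination in a transaction or to a new preset of it.
# Links save per source in asset user data on destination MetaSound Sources for later reload or deletion.
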EditorUtilityBlueprint /TechAudioToolsContent/Tools/MetaSoundInputMigrator/Viewmodels/BP_MetaSoundMemberMigratorViewmodel
graphs [
 Are Both Assets Set {nodes [
   Are Both Assets Set() {is=title K2Node_FunctionEntry}
   n1 = Get Destination Metasound() -> (Destination Metasound:Meta Sound Document Interface Interface)
   n2 = Get Source Metasound() -> (Source Metasound:Meta Sound Document Interface Interface)
   > Return Node!(
    execute
    ReturnValue:bool=AND Boolean(A=IsValidInterface!(Interface:"Interface Interface (by ref)"=@n2) -> (ReturnValue:bool), B=IsValidInterface(Interface=@n1))
   )
  ]}
 Broadcast Fields {nodes [
   Broadcast Fields() {is=title K2Node_FunctionEntry}
   n1 = Self-Reference()
   > Broadcast Field Value Changed!(execute, Object:Object Reference=@n1, FieldId:Field Notification Id Structure=(FieldName="Can Link Selected Inputs")) -> (then)
   > Broadcast Field Value Changed(Object=@n1, FieldId=(FieldName="Can Unlink Selected Inputs"))
   > Broadcast Field Value Changed(Object=@n1, FieldId=(FieldName="Get Number of Linked Inputs"))
   > Broadcast Field Value Changed(Object=@n1, FieldId=(FieldName="Can Create New Preset"))
   > Broadcast Field Value Changed(Object=@n1, FieldId=(FieldName="Can Overwrite Destination Inputs"))
   > Broadcast Field Value Changed(Object=@n1, FieldId=(FieldName="Has Linked Inputs"))
   > Broadcast Field Value Changed(Object=@n1, FieldId=(FieldName="Has Linked Inputs"))
  ]}
 Can Create New Preset {nodes [
   Can Create New Preset() {is=title K2Node_FunctionEntry}
   n1 = Get Linked Inputs() -> (Linked Inputs:Set of Meta Sound Member Migrator Node Link Structures)
   > Return Node(ReturnValue="integer > integer"(A=Length!(TargetSet:Set of Meta Sound Member Migrator Node Link Structures=@n1) -> (ReturnValue:int=0) {is=title K2Node_CallFunction}))
  ]}
 Can Delete Saved Link Data {nodes [
   Can Delete Saved Link Data() {is=title K2Node_FunctionEntry}
   > n1 = Cast To MetaSoundSource!(execute, Object:Meta Sound Document Interface Interface=Get Destination Metasound()) -> (then, CastFailed, AsMetaSound Source:Meta Sound Source Object Reference) {is=title K2Node_DynamicCast}
   > n2 = GetAssetUserDataOfClass!(
    execute
    self:Interface Asset User Data Interface=@n1
    InUserDataClass:Asset User Data Class Reference=/Script/TechAudioToolsContentMetaSoundEditor.MetaSoundMemberMigratorData
   ) -> (then, ReturnValue:Meta Sound Member Migrator Data Object Reference)
   n3 = Get Linked Nodes for MetaSound!(
    self:Meta Sound Member Migrator Data Object Reference=@n2
    SourceMetaSound:"Meta Sound Document Interface Interface (by ref)"=Get Source Metasound()
   ) -> (OutLinkedNodes:Set of Meta Sound Member Migrator Node Link Structures, ReturnValue:bool) {is=title K2Node_CallFunction}
   > Branch(Condition=@n3.ReturnValue)
   > Return Node(ReturnValue="integer > integer"(A=Length(TargetSet=@n3) -> (ReturnValue=0)))
  ]}
 Can Link Selected Inputs {nodes [
   Can Link Selected Inputs() {is=title K2Node_FunctionEntry}
   > n1 = Sequence()
   > n2 = Get!(execute:Exec) -> (then:Exec, else:Exec, Selected Source Input Viewmodel:Meta Sound Input Editor Viewmodel Object Reference)
   Return Node(execute=@n2.else, ReturnValue=false)
   n3 = Sequence(execute=@n1.then_1)
   > n4 = Get!(execute) -> (then, else, Selected Destination Input Viewmodel:Meta Sound Input Editor Viewmodel Object Reference)
   Return Node(execute=@n4.else, ReturnValue=false)
   n5 = Sequence(execute=@n3.then_1)
   n6 = Get Destination Metasound()
   n7 = Get Source Metasound()
   > Branch(Condition="Equal (Object)"(
     A="To Object (Interface)"!(Interface:"Interface Interface (by ref)"=@n7) -> (ReturnValue:Object Reference)
     B="To Object (Interface)"(Interface=@n6)
    ))
   > Return Node(ReturnValue=false)
   n8 = Sequence(execute=@n5.then_1)
   n9 = Get Selected Source Input Viewmodel() -> (Selected Source Input Viewmodel:Meta Sound Input Editor Viewmodel Object Reference)
   n10 = Get Selected Destination Input Viewmodel() -> (Selected Destination Input Viewmodel:Meta Sound Input Editor Viewmodel Object Reference)
   > Branch(Condition="Not Equal (Name)"(A=Get DataType!(self:Meta Sound Input Viewmodel Object Reference=@n9) -> (DataType:Name=None), B=Get DataType(self=@n10) -> (DataType=None)) -> (ReturnValue=~))
   > Return Node(ReturnValue=false)
   Sequence(execute=@n8.then_1)
   n11 = Get Selected Input Node Pair!(self) -> (MetaSoundMemberMigratorNodeLink:Meta Sound Member Migrator Node Link Structure) {is=title K2Node_CallFunction}
   n12 = Get Linked Inputs()
   > Return Node(ReturnValue=NOT Boolean(A=Contains Item!(
      TargetSet:Set of Meta Sound Member Migrator Node Link Structures=@n12
      ItemToFind:"Meta Sound Member Migrator Node Link Structure (by ref)"=@n11
     ) -> (ReturnValue:bool) {is=title K2Node_CallFunction}))
  ]}
 Can Overwrite Destination Inputs {nodes [
   Can Overwrite Destination Inputs() {is=title K2Node_FunctionEntry}
   n1 = Get Linked Inputs()
   > Return Node(ReturnValue="integer > integer"(A=Length(TargetSet=@n1) -> (ReturnValue=0)))
  ]}
 Can Reload Saved Links {nodes [
   Can Reload Saved Links() {is=title K2Node_FunctionEntry}
   > n1 = Cast To MetaSoundSource(Object=Get Destination Metasound())
   > n2 = HasAssetUserDataOfClass!(
    execute
    self:Interface Asset User Data Interface=@n1
    InUserDataClass:Asset User Data Class Reference=&MetaSoundMemberMigratorData
   ) -> (then, ReturnValue:bool)
   > Return Node(ReturnValue=AND Boolean!(A=@n2, B=Are Both Assets Set!(self) -> (ReturnValue:bool) {is=title K2Node_CallFunction}, C:bool=Get Is Dirty() -> (Is Dirty:bool)) -> (ReturnValue))
  ]}
 Can Save Linked Inputs {nodes [
   Can Save Linked Inputs() {is=title K2Node_FunctionEntry}
   n1 = Get Destination Metasound()
   n2 = Get Source Metasound()
   > Return Node(ReturnValue=AND Boolean(
     A=Cast To MetaSoundSource!(Object=@n2) -> (AsMetaSound Source, bSuccess:bool=~).bSuccess
     B=Cast To MetaSoundSource(Object=@n1) -> (bSuccess=~).bSuccess
     C=Get Is Dirty()
    ))
  ]}
 Can Unlink Selected Inputs {nodes [
   Can Unlink Selected Inputs() {is=title K2Node_FunctionEntry}
   > n1 = Sequence()
   > n2 = Get!(execute) -> (then, else, Selected Source Input Viewmodel:Meta Sound Input Editor Viewmodel Object Reference)
   Return Node(execute=@n2.else, ReturnValue=false)
   n3 = Sequence(execute=@n1.then_1)
   > n4 = Get!(execute) -> (then, else, Selected Destination Input Viewmodel:Meta Sound Input Editor Viewmodel Object Reference)
   Return Node(execute=@n4.else, ReturnValue=false)
   Sequence(execute=@n3.then_1)
   > Return Node(ReturnValue=Contains Item(TargetSet=Get Linked Inputs(), ItemToFind=Get Selected Input Node Pair()))
  ]}
 Create New Preset {nodes [
   Create New Preset() {is=title K2Node_FunctionEntry}
   n1 = Get Destination Metasound()
   > n2 = Cast To MetaSoundSource!(execute:Exec, Object=@n1) -> (then:Exec, CastFailed:Exec, AsMetaSound Source)
   n3 = Get MetaSoundBuilderSubsystem() -> (ReturnValue:Meta Sound Builder Subsystem Object Reference) {is=title K2Node_GetEngineSubsystem}
   > n4 = CreateSourcePresetBuilder!(
    execute
    self:Meta Sound Builder Subsystem Object Reference=@n3
    BuilderName:Name=MetaSoundInputMigrator
    ReferencedSourceClass:"Meta Sound Document Interface Interface (by ref)"=@n2
   ) -> (Succeeded:Exec, Failed:Exec, ReturnValue:Meta Sound Source Builder Object Reference)
   > n5 = Set Destination Preset Builder(Destination Preset Builder:Meta Sound Builder Base Object Reference=@n4) -> (Output_Get:Meta Sound Builder Base Object Reference)
   n6 = "For Each Loop (Set)"!(execute:Exec=[@n5.then, @n25.then], SetPin:Set of Meta Sound Member Migrator Node Link Structures=Get Linked Inputs(), BreakPin:Exec) -> (then:Exec, ValuePin:Meta Sound Member Migrator Node Link Structure, CompletedPin:Exec) {is=title K2Node_SetForEach}
   > Set Link(Link:Meta Sound Member Migrator Node Link Structure=@n6) -> (Output_Get:Meta Sound Member Migrator Node Link Structure)
   > n7 = Sequence()
   n8 = Get Link() -> (Link:Meta Sound Member Migrator Node Link Structure)
   n9 = Get MetaSoundBuilderSubsystem()
   n10 = Get Source Metasound()
   > Set Source Literal(Source Literal:Meta Sound Literal Structure=GetGraphInputDefault!(
     self:Meta Sound Builder Base Object Reference=Find Builder By MetaSound!(self:Meta Sound Builder Subsystem Object Reference=@n9, InMetaSound:Meta Sound Document Interface Interface=@n10) -> (ReturnValue:Meta Sound Builder Base Object Reference)
     InputName:Name=Break Meta Sound Member Migrator Node Link!(MetaSoundMemberMigratorNodeLink:"Meta Sound Member Migrator Node Link Structure (by ref)"=@n8) -> (SourceInputName:Name) {is=title K2Node_BreakStruct}
     PageName:Name
    ) -> (ReturnValue:Meta Sound Literal Structure)) -> (Output_Get:Meta Sound Literal Structure)
   n11 = Sequence(execute=@n7.then_1)
   n12 = Get Link()
   n13 = Get MetaSoundBuilderSubsystem()
   n14 = Get Destination Metasound()
   > Set Destination Literal(Destination Literal:Meta Sound Literal Structure=GetGraphInputDefault(
     self=Find Builder By MetaSound(self=@n13, InMetaSound=@n14)
     InputName=Break Meta Sound Member Migrator Node Link!(MetaSoundMemberMigratorNodeLink=@n12) -> (DestinationInputName:Name)
    )) -> (Output_Get:Meta Sound Literal Structure)
   Sequence(execute=@n11.then_1)
   n15 = Get Source Literal() -> (Source Literal:Meta Sound Literal Structure)
   n16 = Get Destination Literal() -> (Destination Literal:Meta Sound Literal Structure)
   > Branch(Condition=NOT Boolean(A="Equal (MetaSound Literal)"(A:"Meta Sound Literal Structure (by ref)"=@n15, B:"Meta Sound Literal Structure (by ref)"=@n16) -> (ReturnValue:bool=~) {is=title K2Node_PromotableOperator}))
   > SetGraphInputDefault!(
    execute
    self:Meta Sound Builder Base Object Reference=Get Destination Preset Builder() -> (Destination Preset Builder:Meta Sound Builder Base Object Reference)
    InputName:Name=Break Meta Sound Member Migrator Node Link(MetaSoundMemberMigratorNodeLink=Get Link())
    Literal:"Meta Sound Literal Structure (by ref)"=Get Source Literal()
    PageName:Name
   ) -> (Succeeded:Exec, Failed:Exec)
   n17 = GetAssetTools(execute=@n6.CompletedPin)
   n18 = "To Object (Interface)"(Interface=Get Destination Metasound())
   n19 = Get Destination Metasound()
   n20 = Split!(InPath:String=Get Object Path String(Object="To Object (Interface)"(Interface=@n19))) -> (PathPart:String, FilenamePart:String, ExtensionPart:String)
   > n21 = CreateUniqueAssetName!(
    execute
    self:Asset Tools Interface=@n17
    InBasePackageName:String=Append!(A=@n20, B=/, C:String=GetDisplayName(Object=@n18)) -> (ReturnValue)
    InSuffix:String
   ) -> (then, OutPackageName:String, OutAssetName:String)
   n22 = GetCurrentContentBrowserPath!() -> (OutPath:String, ReturnValue:bool)
   > BuildToAsset!(
    execute
    self:Meta Sound Editor Subsystem Object Reference=Get MetaSoundEditorSubsystem() -> (ReturnValue:Meta Sound Editor Subsystem Object Reference) {is=title K2Node_GetEditorSubsystem}
    InBuilder:Meta Sound Builder Base Object Reference=Get Destination Preset Builder()
    Author:String=GetPlatformUserName!() -> (ReturnValue:String)
    AssetName:String=@n21.OutAssetName
    PackagePath:String=Select(Option 0:String=@n22, Option 1:String=@n20, Index=@n22.ReturnValue) -> (ReturnValue:String)
    TemplateSoundWave:Sound Wave Object Reference
   ) -> (Succeeded:Exec, Failed:Exec, ReturnValue:Meta Sound Document Interface Interface)
   n23 = Cast To MetaSoundPatch!(execute=@n2.CastFailed, Object:Meta Sound Document Interface Interface=@n1) -> (then, CastFailed, AsMetaSound Patch:Meta Sound Patch Object Reference) {is=title K2Node_DynamicCast}
   > n24 = CreatePatchPresetBuilder!(
    execute
    self:Meta Sound Builder Subsystem Object Reference=@n3
    BuilderName:Name=MetaSoundInputMigrator
    ReferencedPatchClass:"Meta Sound Document Interface Interface (by ref)"=@n23
   ) -> (Succeeded:Exec, Failed:Exec, ReturnValue:Meta Sound Patch Builder Object Reference)
   > n25 = Set Destination Preset Builder(Destination Preset Builder=@n24)
  ]}
 Delete Saved Link Data {nodes [
   Delete Saved Link Data() {is=title K2Node_FunctionEntry}
   > n1 = Cast To MetaSoundSource(Object=Get Destination Metasound())
   > n2 = GetAssetUserDataOfClass(self=@n1, InUserDataClass=&MetaSoundMemberMigratorData)
   > n3 = RemoveLinksToSource!(
    execute
    self:Meta Sound Member Migrator Data Object Reference=@n2
    SourceMetaSound:"Meta Sound Document Interface Interface (by ref)"=Get Source Metasound()
   ) -> (then, ReturnValue:bool)
   > Branch(Condition=@n3)
   > SaveLoadedAsset(AssetToSave="To Object (Interface)"(Interface=Get Destination Metasound()), bOnlyIfIsDirty=false)
   > Unlink All Inputs()
  ]}
 EventGraph {links [], nodes []}
 Get Linked Inputs {nodes [
   Get Linked Inputs() {is=title K2Node_FunctionEntry}
   > Return Node!(execute, Linked Inputs:Set of Meta Sound Member Migrator Node Link Structures=Get Linked Inputs() {is=title K2Node_VariableGet})
  ]}
 Get Number of Linked Inputs {nodes [
   Get Number of Linked Inputs() {is=title K2Node_FunctionEntry}
   > Return Node!(execute, Number:int=Length(TargetSet=Get Linked Inputs()) -> (ReturnValue=0))
  ]}
 Get Selected Input Node Pair {nodes [
   Get Selected Input Node Pair() {is=title K2Node_FunctionEntry}
   n1 = Get Selected Destination Input Viewmodel()
   n2 = Get Selected Source Input Viewmodel()
   > Return Node!(
    execute
    MetaSoundMemberMigratorNodeLink:Meta Sound Member Migrator Node Link Structure=Make Meta Sound Member Migrator Node Link!(
     SourceInputName:Name=Get InputName!(self:Meta Sound Input Viewmodel Object Reference=@n2) -> (InputName:Name=None)
     DestinationInputName:Name=Get InputName(self=@n1) -> (InputName=None)
    ) -> (MetaSoundMemberMigratorNodeLink:Meta Sound Member Migrator Node Link Structure) {is=title K2Node_MakeStruct}
   )
  ]}
 Get or Create Migrator Data {nodes [
   Get or Create Migrator Data() {is=title K2Node_FunctionEntry}
   > n1 = Cast To MetaSoundSource(Object=Get Destination Metasound())
   > n2 = GetAssetUserDataOfClass(self=@n1, InUserDataClass=&MetaSoundMemberMigratorData)
   > n3 = Set Member Migrator Data(Member Migrator Data:Meta Sound Member Migrator Data Object Reference=@n2) -> (Output_Get:Meta Sound Member Migrator Data Object Reference)
   > n4 = Is Valid(InputObject=@n3)
   n5 = AddAssetUserDataOfClass!(
    execute=@n4.Is Not Valid
    self:Interface Asset User Data Interface=Cast To MetaSoundSource!(Object=Get Destination Metasound()) -> (AsMetaSound Source, bSuccess:bool=~)
    InUserDataClass:Asset User Data Class Reference=&MetaSoundMemberMigratorData
   ) -> (then, ReturnValue:bool)
   > n6 = Branch(Condition=@n5)
   > n7 = GetAssetUserDataOfClass(
    self=Cast To MetaSoundSource(Object=Get Destination Metasound()) -> (bSuccess=~)
    InUserDataClass=&MetaSoundMemberMigratorData
   )
   > Return Node!(execute, Migrator Data:Meta Sound Member Migrator Data Object Reference=@n7)
   RaiseScriptError!(execute=@n6.else, ErrorMessage:String=Failed to add asset user data) -> (then)
   > Return Node()
   Return Node(execute=@n4.Is Valid, Migrator Data=@n3)
  ]}
 Has Asset User Data {nodes [
   n1 = Has Asset User Data!() -> (then, Metasound:Meta Sound Document Interface Interface) {is=title K2Node_FunctionEntry}
   > n2 = Cast To MetaSoundSource!(execute:Exec, Object=@n1) -> (then:Exec, CastFailed:Exec, AsMetaSound Source)
   > n3 = GetAssetUserDataOfClass(self=@n2, InUserDataClass=&MetaSoundMemberMigratorData)
   > Return Node!(execute, ReturnValue:bool=IsValid(Object=@n3))
  ]}
 Has Linked Inputs {nodes [
   Has Linked Inputs() {is=title K2Node_FunctionEntry}
   n1 = Get Linked Inputs()
   > Return Node(ReturnValue="integer > integer"(A=Length(TargetSet=@n1) -> (ReturnValue=0)))
  ]}
 Link Inputs {nodes [
   Link Inputs!() -> (then, Inputs to Link:"Meta Sound Member Migrator Node Link Structure (by ref)") {is=title K2Node_FunctionEntry}
   > n1 = Sequence()
   > Add!(
    execute
    TargetSet:Set of Meta Sound Member Migrator Node Link Structures=Get Linked Inputs()
    NewItem:"Meta Sound Member Migrator Node Link Structure (by ref)"=Get Inputs to Link() -> (Inputs to Link:"Meta Sound Member Migrator Node Link Structure (by ref)")
   ) -> (then) {is=title K2Node_CallFunction}
   n2 = Sequence(execute=@n1.then_1)
   n3 = Break Meta Sound Member Migrator Node Link!(MetaSoundMemberMigratorNodeLink=Get Inputs to Link()) -> (SourceInputName:Name)
   > n4 = Cast To BP_MetaSoundInputMigratorEntryViewModel!(
    execute
    Object=FindEntryViewModelByInputName!(
     self:Meta Sound Input List Editor Viewmodel Object Reference=Get Source Migrator List Viewmodel() -> (Source Migrator List Viewmodel:BP Meta Sound Input Migrator List View Model Object Reference)
     InputName:Name=@n3
    ) -> (ReturnValue:Tech Audio Tools Entry Viewmodel Object Reference)
   ) -> (then, CastFailed, AsBP Meta Sound Input Migrator Entry View Model:BP Meta Sound Input Migrator Entry View Model Object Reference) {is=title K2Node_DynamicCast}
   > n5 = Set Source Input(Source Input:BP Meta Sound Input Migrator Entry View Model Object Reference=@n4) -> (Output_Get:BP Meta Sound Input Migrator Entry View Model Object Reference)
   > n6 = Is Valid(InputObject=@n5)
   PrintText(
    execute=@n6.Is Not Valid
    WorldContextObject=Self-Reference()
    InText=Format Text!(Format="LOCTEXT(\"Unable to find input named \\'{input}\\'. Skipping link.\")", input:Name=@n3) -> (Result)
   )
   > Return Node!(execute)
   n7 = Sequence(execute=@n2.then_1)
   n8 = Break Meta Sound Member Migrator Node Link!(MetaSoundMemberMigratorNodeLink=Get Inputs to Link()) -> (DestinationInputName:Name)
   > n9 = Cast To BP_MetaSoundInputMigratorEntryViewModel(Object=FindEntryViewModelByInputName(
     self=Get Destination Migrator List Viewmodel() -> (Destination Migrator List Viewmodel:BP Meta Sound Input Migrator List View Model Object Reference)
     InputName=@n8
    ))
   > n10 = Set Destination Input(Destination Input:BP Meta Sound Input Migrator Entry View Model Object Reference=@n9) -> (Output_Get:BP Meta Sound Input Migrator Entry View Model Object Reference)
   > n11 = Is Valid(InputObject=@n10)
   PrintText(
    execute=@n11.Is Not Valid
    WorldContextObject=Self-Reference()
    InText=Format Text(Format="LOCTEXT(\"Unable to find input named \\'{input}\\'. Skipping link.\")", input=@n8)
   )
   > Return Node()
   n12 = Sequence(execute=@n7.then_1)
   > n13 = Get!(execute) -> (then, else, Source Input:BP Meta Sound Input Migrator Entry View Model Object Reference)
   > n14 = Get!(execute, self:Tech Audio Tools Entry Viewmodel Object Reference=@n13) -> (then, else, EntryData:Object Reference)
   > n15 = Cast To MetaSoundInputEditorViewModel!(execute, Object=@n14) -> (then, CastFailed, AsMetaSound Input Editor Viewmodel:Meta Sound Input Editor Viewmodel Object Reference) {is=title K2Node_DynamicCast}
   > Set Source Data Type(Source Data Type:Name=Get DataType(self=@n15) -> (DataType=None)) -> (Output_Get:Name=None)
   n16 = Sequence(execute=@n12.then_1)
   > n17 = Get!(execute) -> (then, else, Destination Input:BP Meta Sound Input Migrator Entry View Model Object Reference)
   > n18 = Get!(execute, self:Tech Audio Tools Entry Viewmodel Object Reference=@n17) -> (then, else, EntryData:Object Reference)
   > n19 = Cast To MetaSoundInputEditorViewModel(Object=@n18)
   > Set Destination Data Type(Destination Data Type:Name=Get DataType(self=@n19) -> (DataType=None)) -> (Output_Get:Name=None)
   n20 = Sequence(execute=@n16.then_1)
   > Branch(Condition="Not Equal (Name)"(A=Get Source Data Type() -> (Source Data Type:Name=None), B=Get Destination Data Type() -> (Destination Data Type:Name=None)) -> (ReturnValue=~))
   > Return Node()
   n21 = Sequence(execute=@n20.then_1)
   > n22 = Get!(execute) -> (then, else, Destination Input:BP Meta Sound Input Migrator Entry View Model Object Reference)
   > n23 = Get!(execute, self:BP Meta Sound Input Migrator Entry View Model Object Reference=@n22) -> (then, else, Linked Migrator Entry Viewmodel:BP Meta Sound Input Migrator Entry View Model Object Reference)
   > Set with Broadcast Linked Migrator Entry Viewmodel!(
    execute
    Linked Migrator Entry Viewmodel:BP Meta Sound Input Migrator Entry View Model Object Reference
    self:BP Meta Sound Input Migrator Entry View Model Object Reference=@n23
   ) -> (then, Output_Get:BP Meta Sound Input Migrator Entry View Model Object Reference)
   n24 = Sequence(execute=@n21.then_1)
   > n25 = Get!(execute) -> (then, else, Source Input:BP Meta Sound Input Migrator Entry View Model Object Reference)
   > n26 = Get!(execute, self:BP Meta Sound Input Migrator Entry View Model Object Reference=@n25) -> (then, else, Linked Migrator Entry Viewmodel:BP Meta Sound Input Migrator Entry View Model Object Reference)
   > Set with Broadcast Linked Migrator Entry Viewmodel(self=@n26)
   n27 = Sequence(execute=@n24.then_1)
   > n28 = Get!(execute) -> (then, else, Source Input:BP Meta Sound Input Migrator Entry View Model Object Reference)
   > Set with Broadcast Linked Migrator Entry Viewmodel(
    Linked Migrator Entry Viewmodel=Get Destination Input() -> (Destination Input:BP Meta Sound Input Migrator Entry View Model Object Reference)
    self=@n28
   )
   Sequence(execute=@n27.then_1)
   > n29 = Get!(execute) -> (then, else, Destination Input:BP Meta Sound Input Migrator Entry View Model Object Reference)
   > Set with Broadcast Linked Migrator Entry Viewmodel(
    Linked Migrator Entry Viewmodel=Get Source Input() -> (Source Input:BP Meta Sound Input Migrator Entry View Model Object Reference)
    self=@n29
   )
  ]}
 Link Selected Inputs {nodes [
   Link Selected Inputs() {is=title K2Node_FunctionEntry}
   > Link Inputs!(
    execute
    self
    Inputs to Link:"Meta Sound Member Migrator Node Link Structure (by ref)"=Get Selected Input Node Pair() {is=title K2Node_CallFunction}
   ) -> (then) {is=title K2Node_CallFunction}
   > Set with Broadcast Is Dirty!(execute, Is Dirty:bool=true) -> (then, Output_Get:bool)
   > Broadcast Fields() {is=title K2Node_CallFunction}
  ]}
 Load Saved Linked Inputs {nodes [
   Load Saved Linked Inputs() {is=title K2Node_FunctionEntry}
   > Unlink All Inputs()
   > Branch(Condition=Are Both Assets Set())
   > Branch(Condition=Has Asset User Data!(self, Metasound:Meta Sound Document Interface Interface=Get Destination Metasound()) -> (ReturnValue:bool) {is=title K2Node_CallFunction})
   n1 = Get Linked Nodes for MetaSound(
    self=Get or Create Migrator Data!() -> (Migrator Data:Meta Sound Member Migrator Data Object Reference) {is=title K2Node_CallFunction}
    SourceMetaSound=Get Source Metasound()
   )
   > Branch(Condition=@n1.ReturnValue)
   > n2 = Set with Broadcast Linked Inputs!(execute, Linked Inputs:Set of Meta Sound Member Migrator Node Link Structures=@n1) -> (then, Output_Get:Set of Meta Sound Member Migrator Node Link Structures)
   > n3 = "For Each Loop (Set)"(SetPin=@n2)
   > Link Inputs(Inputs to Link=@n3)
   Set with Broadcast Is Dirty(execute=@n3.CompletedPin, Is Dirty=false)
   > Broadcast Fields()
  ]}
 Overwrite Destination Inputs {nodes [
   Overwrite Destination Inputs() {is=title K2Node_FunctionEntry}
   > BeginTransaction(Context=MetaSound Input Migrator, Description="LOCTEXT(\"Overwrite MetaSound input defaults\")")
   > TransactObject(Object="To Object (Interface)"(Interface=Get Destination Metasound()))
   > n1 = "For Each Loop (Set)"(SetPin=Get Linked Inputs())
   > Set Link(Link=@n1)
   > n2 = Sequence()
   n3 = Get Link()
   n4 = Get MetaSoundBuilderSubsystem()
   n5 = Get Source Metasound()
   > Set Source Literal(Source Literal=GetGraphInputDefault(
     self=Find Builder By MetaSound(self=@n4, InMetaSound=@n5)
     InputName=Break Meta Sound Member Migrator Node Link!(MetaSoundMemberMigratorNodeLink=@n3) -> (SourceInputName:Name)
    ))
   n6 = Sequence(execute=@n2.then_1)
   n7 = Get Link()
   n8 = Get MetaSoundBuilderSubsystem()
   n9 = Get Source Metasound()
   > Set Destination Literal(Destination Literal=GetGraphInputDefault(
     self=Find Builder By MetaSound(self=@n8, InMetaSound=@n9)
     InputName=Break Meta Sound Member Migrator Node Link!(MetaSoundMemberMigratorNodeLink=@n7) -> (DestinationInputName:Name)
    ))
   Sequence(execute=@n6.then_1)
   n10 = Get Destination Literal()
   n11 = Get Source Literal()
   > Branch(Condition=NOT Boolean(A="Equal (MetaSound Literal)"(A=@n11, B=@n10) -> (ReturnValue=~)))
   > SetGraphInputDefault(
    self=Find Builder By MetaSound(self=Get MetaSoundBuilderSubsystem(), InMetaSound=Get Destination Metasound())
    InputName=Break Meta Sound Member Migrator Node Link(MetaSoundMemberMigratorNodeLink=Get Link())
    Literal=Get Source Literal()
   )
   EndTransaction(execute=@n1.CompletedPin)
  ]}
 Save Linked Inputs {nodes [
   Save Linked Inputs() {is=title K2Node_FunctionEntry}
   > BeginTransaction(Context=MetaSound Input Migrator, Description="LOCTEXT(\"Save linked MetaSound inputs\")")
   > TransactObject(Object="To Object (Interface)"(Interface=Get Destination Metasound()))
   > SaveLinks!(
    execute
    self:Meta Sound Member Migrator Data Object Reference=Get or Create Migrator Data()
    SourceMetaSound:"Meta Sound Document Interface Interface (by ref)"=Get Source Metasound()
    LinkedNodes:Set of Meta Sound Member Migrator Node Link Structures=Get Linked Inputs()
   ) -> (then)
   > SaveLoadedAsset(AssetToSave="To Object (Interface)"(Interface=Get Destination Metasound()), bOnlyIfIsDirty=false)
   > EndTransaction()
   > Set with Broadcast Is Dirty(Is Dirty=false)
  ]}
 Set Destination Metasound {nodes [
   n1 = Set Destination Metasound!() -> (then, Destination Metasound:Meta Sound Document Interface Interface) {is=title K2Node_FunctionEntry}
   > Set with Broadcast Destination Metasound!(execute, Destination Metasound:Meta Sound Document Interface Interface=@n1) -> (then, Output_Get:Meta Sound Document Interface Interface)
  ]}
 Set Source Metasound {nodes [
   n1 = Set Source Metasound!() -> (then, Source Metasound:Meta Sound Document Interface Interface) {is=title K2Node_FunctionEntry}
   > Set with Broadcast Source Metasound!(execute, Source Metasound:Meta Sound Document Interface Interface=@n1) -> (then, Output_Get:Meta Sound Document Interface Interface)
  ]}
 Unlink All Inputs {nodes [
   Unlink All Inputs() {is=title K2Node_FunctionEntry}
   > n1 = "For Each Loop (Set)"(SetPin=Get Linked Inputs())
   > Unlink Inputs!(execute, self, Inputs to Unlink:"Meta Sound Member Migrator Node Link Structure (by ref)"=@n1) -> (then)
   Clear!(execute=@n1.CompletedPin, TargetSet:Set of Meta Sound Member Migrator Node Link Structures=Get Linked Inputs()) -> (then) {is=title K2Node_CallFunction}
   > Set with Broadcast Is Dirty(Is Dirty=true)
   > Broadcast Fields()
  ]}
 Unlink Inputs {nodes [
   Unlink Inputs!() -> (then, Inputs to Unlink:"Meta Sound Member Migrator Node Link Structure (by ref)") {is=title K2Node_FunctionEntry}
   > n1 = Sequence()
   n2 = Get Inputs to Unlink() -> (Inputs to Unlink:"Meta Sound Member Migrator Node Link Structure (by ref)")
   > n3 = Cast To BP_MetaSoundInputMigratorEntryViewModel(Object=FindEntryViewModelByInputName(
     self=Get Source Migrator List Viewmodel()
     InputName=Break Meta Sound Member Migrator Node Link!(MetaSoundMemberMigratorNodeLink=@n2) -> (SourceInputName:Name)
    ))
   > Set Source Input(Source Input=@n3)
   n4 = Sequence(execute=@n1.then_1)
   n5 = Get Inputs to Unlink()
   > n6 = Cast To BP_MetaSoundInputMigratorEntryViewModel(Object=FindEntryViewModelByInputName(
     self=Get Destination Migrator List Viewmodel()
     InputName=Break Meta Sound Member Migrator Node Link!(MetaSoundMemberMigratorNodeLink=@n5) -> (DestinationInputName:Name)
    ))
   > Set Destination Input(Destination Input=@n6)
   n7 = Sequence(execute=@n4.then_1)
   > n8 = Get!(execute) -> (then, else, Source Input:BP Meta Sound Input Migrator Entry View Model Object Reference)
   > n9 = Get!(execute, self:BP Meta Sound Input Migrator Entry View Model Object Reference=@n8) -> (then, else, Linked Migrator Entry Viewmodel:BP Meta Sound Input Migrator Entry View Model Object Reference)
   > Set with Broadcast Linked Migrator Entry Viewmodel(self=@n9)
   n10 = Sequence(execute=@n7.then_1)
   > n11 = Get!(execute) -> (then, else, Destination Input:BP Meta Sound Input Migrator Entry View Model Object Reference)
   > n12 = Get!(execute, self:BP Meta Sound Input Migrator Entry View Model Object Reference=@n11) -> (then, else, Linked Migrator Entry Viewmodel:BP Meta Sound Input Migrator Entry View Model Object Reference)
   > Set with Broadcast Linked Migrator Entry Viewmodel(self=@n12)
   Sequence(execute=@n10.then_1)
   > Set with Broadcast Linked Migrator Entry Viewmodel(self=Get Source Input())
   > Set with Broadcast Linked Migrator Entry Viewmodel(self=Get Destination Input())
  ]}
 Unlink Selected Inputs {nodes [
   Unlink Selected Inputs() {is=title K2Node_FunctionEntry}
   n1 = Get Selected Input Node Pair()
   > Unlink Inputs(Inputs to Unlink=@n1) {is=title K2Node_CallFunction}
   > n2 = Remove!(
    execute
    TargetSet:Set of Meta Sound Member Migrator Node Link Structures=Get Linked Inputs()
    Item:"Meta Sound Member Migrator Node Link Structure (by ref)"=@n1
   ) -> (then, ReturnValue:bool)
   > Set with Broadcast Is Dirty(Is Dirty=@n2)
   > Broadcast Fields()
  ]}
]
variables [
 Automatically Filter Data Types {type=boolean}
 Destination Metasound {default=None}
 Destination Migrator List Viewmodel {type=BP_MetaSoundInputMigratorListViewModel_C, default=null}
 Is Dirty {type=boolean}
 Linked Inputs {type="array<MetaSoundMemberMigratorNodeLink>", default []}
 Selected Destination Input Viewmodel {type=MetaSoundInputEditorViewModel, default=null}
 Selected Source Input Viewmodel {type=MetaSoundInputEditorViewModel, default=null}
 Source Metasound {default=None}
 Source Migrator List Viewmodel {type=BP_MetaSoundInputMigratorListViewModel_C, default=null}
]
